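; Gameplay ability for the MeatPizza environmental item that checks whether the player is already crouched, apparently on entry.
; On activation it resolves the pawn from the owning player state and, if crouched, starts a 0.25-second timer.
; The timer sends the pawn a gameplay event with an empty tag, noted as missing Event.MeatPizza.AlreadyCrouched, then ends the ability.
; The non-crouched path never calls EndAbility.
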(event Ability|EventActivateAbility
  (bind _asfort_player_state (Utilities|Casting|CastToFortPlayerState (Ability|GetOwningActorfromActorInfo)))
  (if (Variables|/Script/Engine.Character|GetbIsCrouched (Game|Player|GetPlayerPawn (Fortnite|GetFortPlayerControllerfromActor _asfort_player_state)))
    (Utilities|Time|SetTimerbyEvent (EventDispatchers|CreateEvent :!event Custom|WaitAMoment) 0.25)))

(event Custom|WaitAMoment
  (bind _asfort_player_state (Utilities|Casting|CastToFortPlayerState (Ability|GetOwningActorfromActorInfo)))
  ; S10 Extraction - Missing Tag - Event.MeatPizza.AlreadyCrouched
  (Ability|SendGameplayEventtoActor (Game|Player|GetPlayerPawn (Fortnite|GetFortPlayerControllerfromActor _asfort_player_state)) "(TagName=\"\")")
  (Ability|EndAbility))
